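# Animation Blueprint for the Fortnite base cape v3 cosmetic, following the owning character's mesh and controlling the cape's cloth simulation.
# The AnimGraph layers the owner's copied pose from the root, then replaces the root with the owner's world-space root transform, cached on each update.
# Each update also calls the shared cloth asset controller, tracking the pelvis.
# OnInitialUpdate can override a RigidBody node's physics asset, though the AnimGraph shows no such node.
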AnimBlueprint /BRCosmetics/Accessories/FORT_Capes/Common/Base_Cape_v3/Fortnite_Base_Cape_v3_AnimBP
target_skeleton=&/M_MED_Base_Cape_v3_Skeleton
graphs [
 AnimGraph {nodes [
   Output Pose(Result=Use cached pose 'Start_Cache'()) {layerGroup=None}
   n1 = Copy Pose From Mesh(SourceMeshComponent=Get ownerSkelMesh() -> (ownerSkelMesh:Skeletal Mesh Component Object Reference))
   n2 = "Transform (Modify) Bone - Bone: root"(
    ComponentPose=Local To Component(LocalPose=Layered blend per bone(BlendPoses_0=@n1) {layerSetup [{branchFilters [{boneName=root}]}]})
    Translation=Get root_loc() -> (root_loc:Vector)
    Rotation=Get root_rot() -> (root_rot:Rotator)
    Scale=Get root_scale() -> (root_scale:Vector)
   ) {
    rotationMode=BMM_Replace
    rotationSpace=BCS_WorldSpace
    scaleMode=BMM_Replace
    scaleSpace=BCS_WorldSpace
    translationMode=BMM_Replace
    translationSpace=BCS_WorldSpace
    boneToModify {boneName=root}
   }
   Save cached pose 'Start_Cache'(Pose=Component To Local(ComponentPose=@n2)) {cacheName=Start_Cache}
  ]}
 EventGraph {nodes [
   Event BlueprintUpdateAnimation()
   > n1 = Sequence()
   Cloth Asset Controller Function!(
    execute=@n1.then_1
    CharacterPartAnimInstance:Custom Character Part Anim Instance Object Reference=Self-Reference()
    ClothAssetControllerParameters:Fort Cloth Asset Controller State Parameter Modifiers Object Reference=/BRCosmetics/Characters/Player/Common/Cloth/Common_ClothAssetDataParams
    ClothingAssetName:Name=Fortnite_Base_Cape_v3_ClothAsset
    TrackedBoneName:Name=pelvis
    JumpFallingDurationSeconds:float=1.3
   ) -> (then, OutAppliedParameters:Cloth Asset Controller Parameters Structure)
   n2 = BreakTransform(InTransform=GetSocketTransform(self=Get ownerSkelMesh(), InSocketName=root, TransformSpace=RTS_World))
   Set root_loc(execute=@n1.then_0, root_loc:Vector=@n2) -> (Output_Get:Vector)
   > Set root_rot(root_rot:Rotator=@n2.Rotation) -> (Output_Get:Rotator)
   > Set root_scale(root_scale:Vector=@n2.Scale) -> (Output_Get:Vector)
   Event OnCharacterCustomizationLoaded()
   > Set ownerSkelMesh(ownerSkelMesh:Skeletal Mesh Component Object Reference=GetOwnerSkeletalMeshComponent()) -> (Output_Get:Skeletal Mesh Component Object Reference)
   DrawJointDebugFunction!(
    execute
    SkelMeshComponent:Scene Component Object Reference=GetPartSkeletalMeshComponent()
    JointOffset:Vector="0, 0, 0"
    JointName:Array of Names=Make Array("[0]":Name=root) -> (Array:Array of Names)
    Scale:double=10
    Duration:double=0
    Thickness:double=0.7
    DebugDraw:bool=true
    Foreground:EDrawDebugSceneDepthPriorityGroup Enum=Foreground
   ) -> (then)
  ]}
 OnInitialUpdate {nodes [
   n1 = OnInitialUpdate!() -> (then, Context:"Anim Update Context Structure (by ref)", Node:"Anim Node Reference Structure (by ref)") {is=title K2Node_FunctionEntry}
   > n2 = ConvertToRigidBodyAnimNode!(execute, Node:"Anim Node Reference Structure (by ref)"=@n1.Node) -> (Succeeded:Exec, Failed:Exec, ReturnValue:Rigid Body Anim Node Reference Structure)
   > SetOverridePhysicsAsset!(
    execute
    Node:"Rigid Body Anim Node Reference Structure (by ref)"=@n2
    PhysicsAsset:Physics Asset Object Reference=Get Physics Asset() -> (Physics Asset:Physics Asset Object Reference)
   ) -> (then, ReturnValue:Rigid Body Anim Node Reference Structure)
   > Return Node()
  ]}
]
variables [
 BackpackRoot_RotationOffset {type=Vector, default=<0 -20 0>}
 BodyMesh {type=CustomCharacterPart, default=null}
 BwdMovement {type=number}
 Cape_LookAt_Target {type=Vector}
 ChaosClothingInteractor {type=ChaosClothingInteractor, default=null}
 Cloth Core Settings_0 {
  type=ClothCoreSettings
  default {
   collisionThickness=1.5
   dampingCoefficient=0.1
   frictionCoefficient=0.2
   localDampingCoefficient=0
   numberIterations=2
   numberSubsteps=1
   selfCollisionThickness=0
   animDriveDamping=<0 1>
   animDriveStiffness=<0 1>
   areaStiffness=<1 1>
   bendingStiffness=<1 1>
   edgeStiffness=<1 1>
   tetherStiffness=<1 1>
  }
 }
 GravityOverride {type=Vector}
 ownerSkelMesh {type=SkeletalMeshComponent}
 Physics Asset {type=PhysicsAsset, default=/BRCosmetics/Accessories/FORT_Capes/Common/Base_Cape_v2/Fortnite_Base_Cape_v2_Dyn, class=~, bodies=~}
 RbanOrCloth {type=boolean}
 resetClothLandingFromZilpine {type=boolean}
 root_loc {type=Vector}
 root_rot {type=Rotator}
 root_scale {type=Vector}
 Sim Space Settings {
  type=SimSpaceSettings
  default {
   dampingAlpha=1
   maxAngularAcceleration=500
   maxAngularVelocity=3
   maxLinearAcceleration=700
   maxLinearVelocity=500
   velocityScaleZ=0
   worldAlpha=1
   externalLinearDragV=<1 1 0>
  }
 }
 Sim Space Settings_Base {
  type=SimSpaceSettings
  default {
   dampingAlpha=1
   maxAngularAcceleration=500
   maxAngularVelocity=3
   maxLinearAcceleration=700
   maxLinearVelocity=500
   velocityScaleZ=0
   worldAlpha=1
   externalLinearDragV=<1 1 0>
  }
 }
 Sim Space Settings_GrindingWeaponActive {
  type=SimSpaceSettings
  default {
   dampingAlpha=1
   maxAngularAcceleration=500
   maxAngularVelocity=3
   maxLinearAcceleration=700
   maxLinearVelocity=500
   velocityScaleZ=0
   worldAlpha=0.2
   externalLinearDragV=<1 1 0>
  }
 }
 Sim Space Settings_GrindTarget {
  type=SimSpaceSettings
  default {
   dampingAlpha=1
   maxAngularAcceleration=500
   maxAngularVelocity=3
   maxLinearAcceleration=700
   maxLinearVelocity=500
   velocityScaleZ=0
   worldAlpha=0.1
   externalLinearDragV=<1 1 0>
  }
 }
 SkydivingAlpha {type=Rotator}
 TeleportDistanceThreshold_Local {type=Vector}
 Wind Gust Local Settings {type=WindGustSettings, default {elapsedTime=0, finalGustStrengthLocal=0, gustDuration=0, gustStrength=0, timeBetweenGusts=0}}
 Wind Gust Local Settings Out {type=WindGustSettings, default {elapsedTime=0, finalGustStrengthLocal=0, gustDuration=0, gustStrength=0, timeBetweenGusts=0}}
]
